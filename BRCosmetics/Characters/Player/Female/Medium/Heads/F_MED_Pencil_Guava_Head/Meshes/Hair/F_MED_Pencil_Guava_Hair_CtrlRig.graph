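# Control Rig for the Pencil Guava medium female head's hair, adjusting hair master bones as the head moves.
# A spherical pose reader measures head orientation against spine_05, with asymmetric active and falloff regions.
# Its output weights three chained transform modifiers on the right, back and left hair master bones.
# All three share one Euler-derived rotation plus small translations; the side hair bones are not driven.
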ControlRigBlueprint /BRCosmetics/Characters/Player/Female/Medium/Heads/F_MED_Pencil_Guava_Head/Meshes/Hair/F_MED_Pencil_Guava_Hair_CtrlRig
graphs [
 RigVMFunctionLibrary {links [], nodes []}
 RigVMModel {nodes [
   RigUnit_BeginExecution = &RigUnit_BeginExecution()
   n1 = &RigUnit_SphericalPoseReader(
    ActiveRegionScaleFactors=(PositiveWidth=0.5,NegativeWidth=2,PositiveHeight=1,NegativeHeight=1)
    ActiveRegionSize=0.1
    Debug=(bDrawDebug=false,DebugScale=25,DebugSegments=20,DebugThickness=0.25)
    DriverAxis=(X=0,Y=-1,Z=0)
    DriverItem=(Type=Bone,Name="head")
    FalloffRegionScaleFactors=(PositiveWidth=1.5,NegativeWidth=1,PositiveHeight=2.5,NegativeHeight=2.5)
    FalloffSize=0.2
    OptionalParentItem=(Type=Bone,Name="spine_05")
    RotationOffset=(X=-90,Y=0,Z=-70)
    ExecutePin=@RigUnit_BeginExecution.ExecutePin
   )
   n2 = &RigVMFunction_MathQuaternionFromEuler(Euler=(X=-60,Y=0,Z=0))
   > &RigUnit_ModifyTransforms(
    ItemToModify=((Item=(Type=Bone,Name="dyn_mstr_hair_r"),Transform=(Rotation=(X=0,Y=0,Z=0,W=1),Translation=(X=0,Y=3,Z=0),Scale3D=(X=1,Y=1,Z=1))))
    ItemToModify.0.Transform.Rotation=@n2.Result
    Weight=@n1.OutputParam
   )
   > &RigUnit_ModifyTransforms(
    ItemToModify=((Item=(Type=Bone,Name="dyn_mstr_hair_b"),Transform=(Rotation=(X=0,Y=0,Z=0,W=1),Translation=(X=0,Y=-3,Z=0),Scale3D=(X=1,Y=1,Z=1))))
    ItemToModify.0.Transform.Rotation=@n2.Result
    Weight=@n1.OutputParam
   )
   > &RigUnit_ModifyTransforms(
    ItemToModify=((Item=(Type=Bone,Name="dyn_mstr_hair_l"),Transform=(Rotation=(X=0,Y=0,Z=0,W=1),Translation=(X=0,Y=-3,Z=0),Scale3D=(X=1,Y=1,Z=1))))
    ItemToModify.0.Transform.Rotation=@n2.Result
    Weight=@n1.OutputParam
   )
  ]}
]
hierarchy [
 root {loc=<0 0 0>, rot=<0 0 0 1>}
 pelvis {parent=root, loc=<0 -0.5301 98.0571>, rot=<0.0127 0.707 -0.0127 -0.707>}
 spine_01 {parent=pelvis, loc=<6.1208 -0.9328 0>, rot=<0 0 0.034 -0.9994>}
 spine_02 {parent=spine_01, loc=<6.5996 -0.2095 0>, rot=<0 0 0.0238 -0.9997>}
 spine_03 {parent=spine_02, loc=<6.4914 -0.1237 0>, rot=<0 0 0.0149 -0.9999>}
 spine_04 {parent=spine_03, loc=<6.9598 0.0794 0>, rot=<0 0 -0.0132 -0.9999>}
 spine_05 {parent=spine_04, loc=<6.9987 0 0>, rot=<0 0 -0.0015 -1>}
 neck_01 {parent=spine_05, loc=<11.3138 0.1529 0>, rot=<0 0 -0.2003 -0.9797>}
 neck_02 {parent=neck_01, loc=<4.5627 0 0>, rot=<0 0 0.0504 -0.9987>}
 head {parent=neck_02, loc=<5.3914 0 0>, rot=<0 0 0.115 -0.9934>}
 dyn_mstr_side_l {parent=head, loc=<6.0749 2.2326 -7.6696>, rot=<0.0029 -0.7071 -0.0029 -0.7071>}
 dyn_side_l_01 {parent=dyn_mstr_side_l, loc=<0 0 0>, rot=<-0.0972 -0.7055 -0.3184 -0.6256>}
 dyn_side_l_02 {parent=dyn_side_l_01, loc=<3.4132 0 0>, rot=<0 0 0 1>}
 dyn_mstr_hair_l {parent=head, loc=<1.5372 -4.3463 -4.1911>, rot=<0.0029 -0.7071 -0.0029 -0.7071>}
 dyn_hair_l_01 {parent=dyn_mstr_hair_l, loc=<0 0 0>, rot=<-0.4174 -0.542 0.4379 -0.5833>}
 dyn_hair_l_02 {parent=dyn_hair_l_01, loc=<4.4007 0 0>, rot=<-0.001 0.0157 -0.0101 -0.9998>}
 dyn_hair_l_03 {parent=dyn_hair_l_02, loc=<4.3608 0 0>, rot=<0 0 0 1>}
 dyn_hair_l_02_01 {parent=dyn_hair_l_02, loc=<2.232 1.233 -1.4741>, rot=<0.456 -0.0645 -0.2461 -0.8528>}
 dyn_hair_l_01_01 {parent=dyn_hair_l_01, loc=<3.3947 2.7982 -1.9301>, rot=<0.6653 -0.0976 -0.3218 -0.6665>}
 dyn_mstr_hair_b {parent=head, loc=<2.5189 -5.8751 0>, rot=<0.0029 -0.7071 -0.0029 -0.7071>}
 dyn_hair_b_01 {parent=dyn_mstr_hair_b, loc=<0 0 0>, rot=<-0.4835 -0.4835 0.516 -0.516>}
 dyn_hair_b_02 {parent=dyn_hair_b_01, loc=<3.9571 0 0.0227>, rot=<0 0.044 0 -0.999>}
 dyn_hair_b_03_01 {parent=dyn_hair_b_02, loc=<5.3501 0 0>, rot=<0 0 0 1>}
 dyn_hair_b_03_03 {parent=dyn_hair_b_02, loc=<5.7418 1.8571 0.2329>, rot=<0.1541 0.1772 -0.1211 -0.9645>}
 dyn_hair_b_03_02 {parent=dyn_hair_b_02, loc=<5.7418 -1.8571 0.2329>, rot=<-0.1541 0.1772 0.1211 -0.9645>}
 dyn_mstr_side_r {parent=head, loc=<6.0749 2.2326 7.6696>, rot=<0.7071 0.0029 -0.7071 0.0029>}
 dyn_side_r_01 {parent=dyn_mstr_side_r, loc=<0 0 0>, rot=<-0.3184 0.6256 0.0972 -0.7055>}
 dyn_side_r_02 {parent=dyn_side_r_01, loc=<3.4133 0.0001 0>, rot=<0 0 0 -1>}
 dyn_mstr_hair_r {parent=head, loc=<1.4688 -3.9841 3.5306>, rot=<0.7071 0.0029 -0.7071 0.0029>}
 dyn_hair_r_01 {parent=dyn_mstr_hair_r, loc=<0 0 0>, rot=<-0.4735 -0.561 -0.4633 0.4964>}
 dyn_hair_r_02 {parent=dyn_hair_r_01, loc=<4.3999 0 0.0001>, rot=<0.001 0.0157 0.0101 -0.9998>}
 dyn_hair_r_03 {parent=dyn_hair_r_02, loc=<4.3608 0 0>, rot=<0 0 0 1>}
 dyn_hair_r_02_01 {parent=dyn_hair_r_02, loc=<2.2325 -1.233 -1.4741>, rot=<-0.456 -0.0645 0.2461 -0.8528>}
 dyn_hair_r_01_01 {parent=dyn_hair_r_01, loc=<3.3946 -2.7982 -1.9301>, rot=<-0.6653 -0.0976 0.3218 -0.6665>}
 clavicle_l {parent=spine_05, loc=<5.4831 -0.1053 -3.2946>, rot=<0.0465 -0.7289 -0.009 -0.6829>}
 upperarm_l {parent=clavicle_l, loc=<9.8658 -1.2136 -0.4105>, rot=<0.0164 -0.3245 -0.0269 -0.9454>}
 clavicle_r {parent=spine_05, loc=<5.4831 -0.1053 3.2946>, rot=<0.7289 0.0465 -0.6829 0.009>}
 upperarm_r {parent=clavicle_r, loc=<-9.8658 1.2136 0.4105>, rot=<0.0164 -0.3245 -0.0269 -0.9454>}
]
