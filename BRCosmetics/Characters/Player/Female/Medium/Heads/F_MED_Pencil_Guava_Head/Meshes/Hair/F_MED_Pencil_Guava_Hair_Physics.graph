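# Physics asset for the Pencil Guava female head's hair, simulating back, left and right hair bone chains against the upper body.
# Each branching capsule chain hangs from a small master body constrained to the head, with modest swing and tight twist limits along its links.
# Capsules on head, neck, clavicles and upper spine supply body shapes, joined by free angular constraints.
# Free-angular cross constraints tie the middle back body to both side chains.
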PhysicsAsset /BRCosmetics/Characters/Player/Female/Medium/Heads/F_MED_Pencil_Guava_Head/Meshes/Hair/F_MED_Pencil_Guava_Hair_Physics
bodies {
 clavicle_l [{shapeType=Capsule, length=13.25, radius=7.5, shapeName=clavicle_l_capsule, center=<3.7439 0 0>, rotation=<rot 90 0 0>}]
 clavicle_r [{shapeType=Capsule, length=13.25, radius=7, shapeName=clavicle_l_capsule, center=<-3.7439 0 0>, rotation=<rot 90 0 0>}]
 dyn_hair_b_01 [{shapeType=Capsule, length=3, radius=2.2, shapeName=dyn_hair_b_01_capsule, center=<2 0 0.5871>, rotation=<rot 90 0 0>}]
 dyn_hair_b_02 [{shapeType=Capsule, length=5, radius=2, shapeName=dyn_hair_b_02_capsule, center=<2 0 1>, rotation=<rot 90 0 0>}]
 dyn_hair_b_03_01 [{shapeType=Capsule, length=3, radius=2.5, shapeName=dyn_hair_b_03_01_capsule, center=<2 0 2>, rotation=<rot 90 0 0>}]
 dyn_hair_b_03_02 [{shapeType=Capsule, length=3, radius=2.5, shapeName=dyn_hair_b_03_02_capsule, center=<2 0 2>, rotation=<rot 90 0 0>}]
 dyn_hair_b_03_03 [{shapeType=Capsule, length=3, radius=2.5, shapeName=dyn_hair_b_03_03_capsule, center=<2 0 2>, rotation=<rot 90 0 0>}]
 dyn_hair_l_01 [{shapeType=Capsule, length=3, radius=2.2, shapeName=dyn_hair_l_01_capsule, center=<2 0 0.7599>, rotation=<rot 90 0 0>}]
 dyn_hair_l_01_01 [{shapeType=Capsule, length=3.25, radius=2.35, shapeName=dyn_hair_l_01_01_capsule, center=<2 0 1>, rotation=<rot 90 0 0>}]
 dyn_hair_l_02 [{shapeType=Capsule, length=4, radius=2, shapeName=dyn_hair_l_02_capsule, center=<2 0 1>, rotation=<rot 90 0 0>}]
 dyn_hair_l_02_01 [{shapeType=Capsule, length=3.25, radius=2.25, shapeName=dyn_hair_l_02_01_capsule, center=<2 0 1>, rotation=<rot 90 0 0>}]
 dyn_hair_l_03 [{shapeType=Capsule, length=3, radius=2.5, shapeName=dyn_hair_l_03_capsule, center=<2 0 1>, rotation=<rot 90 0 0>}]
 dyn_hair_r_01 [{shapeType=Capsule, length=3, radius=2.5, shapeName=dyn_hair_l_01_capsule, center=<2 0 1.4947>, rotation=<rot 90 0 0>}]
 dyn_hair_r_01_01 [{shapeType=Capsule, length=3.25, radius=2.5, shapeName=dyn_hair_l_01_01_capsule, center=<2.5232 0 1>, rotation=<rot 90 0 0>}]
 dyn_hair_r_02 [{shapeType=Capsule, length=4, radius=2.5, shapeName=dyn_hair_l_02_capsule, center=<2 0 1>, rotation=<rot 90 0 0>}]
 dyn_hair_r_02_01 [{shapeType=Capsule, length=4.3, radius=2.5, shapeName=dyn_hair_l_02_01_capsule, center=<1.7264 0 2>, rotation=<rot 90 0 0>}]
 dyn_hair_r_03 [{shapeType=Capsule, length=3, radius=2.5, shapeName=dyn_hair_l_03_capsule, center=<1.3147 0 2>, rotation=<rot 90 0 0>}]
 dyn_mstr_hair_b [{shapeType=Capsule, length=0.505, radius=0.505, shapeName=dyn_mstr_hair_b_capsule, rotation=<rot 0 -90 0>}]
 dyn_mstr_hair_l [{shapeType=Capsule, length=0.505, radius=0.505, shapeName=dyn_mstr_hair_l_capsule, rotation=<rot 0 -90 0>}]
 dyn_mstr_hair_r [{shapeType=Capsule, length=0.505, radius=0.505, shapeName=dyn_mstr_hair_l_capsule, rotation=<rot 0 -90 0>}]
 head [{shapeType=Capsule, length=15.25, radius=6, shapeName=head_capsule, center=<2.5 1.5 0>, rotation=<rot 90 0 0>}]
 neck_01 [{shapeType=Capsule, length=11.25, radius=4, shapeName=neck_01_capsule, center=<1.3696 1.2786 0>, rotation=<rot 90 0 16.33>}]
 spine_05 [{shapeType=Capsule, length=18, radius=13.5, shapeName=spine_05_capsule, center=<0 0.5 0>, rotation=<rot 0 -90 0>}]
}
constraints [
 {bone1Name=dyn_hair_b_01, bone2Name=dyn_mstr_hair_b, swing1LimitDegrees=15, swing2LimitDegrees=20, twistLimitDegrees=4}
 {bone1Name=dyn_hair_b_02, bone2Name=dyn_hair_b_01, swing1LimitDegrees=20, swing2LimitDegrees=20, twistLimitDegrees=4}
 {bone1Name=dyn_hair_b_03_01, bone2Name=dyn_hair_b_02, swing1LimitDegrees=25, swing2LimitDegrees=25, twistLimitDegrees=2}
 {bone1Name=dyn_hair_b_03_02, bone2Name=dyn_hair_b_02, swing1LimitDegrees=25, swing2LimitDegrees=25, twistLimitDegrees=2}
 {bone1Name=dyn_hair_b_03_03, bone2Name=dyn_hair_b_02, swing1LimitDegrees=25, swing2LimitDegrees=25, twistLimitDegrees=2}
 {bone1Name=dyn_mstr_hair_l, bone2Name=head, swing1LimitDegrees=45, swing2LimitDegrees=45, twistLimitDegrees=2}
 {bone1Name=dyn_hair_l_01, bone2Name=dyn_mstr_hair_l, swing1LimitDegrees=10, swing2LimitDegrees=20, twistLimitDegrees=4}
 {bone1Name=dyn_hair_l_02, bone2Name=dyn_hair_l_01, swing1LimitDegrees=15, swing2LimitDegrees=20, twistLimitDegrees=4}
 {bone1Name=dyn_hair_l_03, bone2Name=dyn_hair_l_02, swing1LimitDegrees=45, swing2LimitDegrees=45, twistLimitDegrees=2}
 {bone1Name=dyn_hair_l_02_01, bone2Name=dyn_hair_l_02, swing1LimitDegrees=45, swing2LimitDegrees=45, twistLimitDegrees=2}
 {bone1Name=dyn_hair_l_01_01, bone2Name=dyn_hair_l_01, swing1LimitDegrees=45, swing2LimitDegrees=45, twistLimitDegrees=2}
 {bone1Name=dyn_mstr_hair_b, bone2Name=head, swing1LimitDegrees=45, swing2LimitDegrees=45, twistLimitDegrees=2}
 {
  bone1Name=head
  bone2Name=neck_01
  swing1LimitDegrees=45
  swing1Motion=Free
  swing2LimitDegrees=45
  swing2Motion=Free
  twistLimitDegrees=45
  twistMotion=Free
 }
 {
  bone1Name=clavicle_l
  bone2Name=spine_05
  swing1LimitDegrees=45
  swing1Motion=Free
  swing2LimitDegrees=45
  swing2Motion=Free
  twistLimitDegrees=45
  twistMotion=Free
 }
 {
  bone1Name=clavicle_r
  bone2Name=spine_05
  swing1LimitDegrees=45
  swing1Motion=Free
  swing2LimitDegrees=45
  swing2Motion=Free
  twistLimitDegrees=45
  twistMotion=Free
 }
 {
  bone1Name=dyn_mstr_hair_r
  bone2Name=head
  swing1LimitDegrees=45
  swing1Motion=Free
  swing2LimitDegrees=45
  swing2Motion=Free
  twistLimitDegrees=45
  twistMotion=Free
 }
 {bone1Name=dyn_hair_r_01, bone2Name=dyn_mstr_hair_r, swing1LimitDegrees=13, swing2LimitDegrees=25, twistLimitDegrees=4}
 {bone1Name=dyn_hair_r_02, bone2Name=dyn_hair_r_01, swing1LimitDegrees=13, swing2LimitDegrees=25, twistLimitDegrees=4}
 {bone1Name=dyn_hair_r_03, bone2Name=dyn_hair_r_02, swing1LimitDegrees=20, swing2LimitDegrees=25, twistLimitDegrees=2}
 {bone1Name=dyn_hair_r_02_01, bone2Name=dyn_hair_r_02, swing1LimitDegrees=20, swing2LimitDegrees=20, twistLimitDegrees=2}
 {bone1Name=dyn_hair_r_01_01, bone2Name=dyn_hair_r_01, swing1LimitDegrees=35, swing2LimitDegrees=35, twistLimitDegrees=2}
 {
  bone1Name=neck_01
  bone2Name=spine_05
  swing1LimitDegrees=45
  swing1Motion=Free
  swing2LimitDegrees=45
  swing2Motion=Free
  twistLimitDegrees=45
  twistMotion=Free
 }
 {
  bone1Name=dyn_hair_r_02
  bone2Name=dyn_hair_b_02
  swing1LimitDegrees=45
  swing1Motion=Free
  swing2LimitDegrees=45
  swing2Motion=Free
  twistLimitDegrees=45
  twistMotion=Free
 }
 {
  bone1Name=dyn_hair_l_02
  bone2Name=dyn_hair_b_02
  swing1LimitDegrees=45
  swing1Motion=Free
  swing2LimitDegrees=45
  swing2Motion=Free
  twistLimitDegrees=45
  twistMotion=Free
 }
 {
  bone1Name=dyn_hair_b_02
  bone2Name=dyn_hair_r_02
  swing1LimitDegrees=45
  swing1Motion=Free
  swing2LimitDegrees=45
  swing2Motion=Free
  twistLimitDegrees=45
  twistMotion=Free
 }
 {
  bone1Name=dyn_hair_b_02
  bone2Name=dyn_hair_l_02
  swing1LimitDegrees=45
  swing1Motion=Free
  swing2LimitDegrees=45
  swing2Motion=Free
  twistLimitDegrees=45
  twistMotion=Free
 }
]
